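# Translucent unlit two-sided material that draws animated aurora curtains as an effect for the Star Walker character body.
# Panned tiling and blurred noise blends two top colors toward a bottom color, varies height and flickers HDR intensity through DeriveHDRfromLDR and emissive adjustments.
# Opacity combines a noise-varied SphereMask, UV gradient powers and squared-sine horizontal falloff.
# A separate panned noise, scaled by WPOScale and transformed from local space, feeds vertex sway.
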Material /BRCosmetics/Characters/Player/Male/Medium/Bodies/M_MED_Star_Walker/FX/M_Auroras_StarWalkerBody
graph {
 properties {blendMode=BLEND_Translucent, pixelDepthOffsetMode=PDOM_Legacy, shadingModel=MSM_Unlit, twoSided=true, usedShadingModels=Unlit}
 nodes [
  n1 = VectorParameter() {parameterName=TopColorA, defaultValue=<rgba 0.345 0.2 2 1>}
  n2 = VectorParameter() {parameterName=TopColorB, defaultValue=<rgba 1 0.15 0.9489 1>}
  n3 = VectorParameter() {parameterName=NoiseTexScale3, defaultValue=<rgba 5 1 0 0>}
  n4 = VertexColor()
  n5 = Multiply(A=TextureCoordinate(), B=AppendVector(A=LinearInterpolate(A=@n3.R, B=@n3.R, Alpha=@n4.B), B=Constant()))
  n6 = Multiply(A=Time(), B=ScalarParameter() {defaultValue=0.25, parameterName=WorldAlignedPanSpeed})
  n7 = ComponentMask(None=TextureSample(UVs=Panner(Coordinate=@n5, Time=@n6) {speedY=1}) {texture=&T_Smooth_tile}.RGB) {g=true}
  n8 = Sine(None=ComponentMask(None=TextureCoordinate()) {r=true}) {period=2}
  n9 = Multiply(A=@n8, B=@n8)
  n10 = TextureCoordinate()
  n11 = VectorParameter() {parameterName=NoiseTexScale, defaultValue=<rgba 5 0.25 0 0>}
  n12 = AppendVector(A=@n11.R, B=@n11.R)
  n13 = TextureSample(UVs=Panner(Coordinate=Multiply(A=@n10, B=@n12), Time=ScalarParameter() {defaultValue=2, parameterName=HeightVariancePanSpeed}) {speedX=0.04, speedY=0.12}) {samplerType=SAMPLERTYPE_Grayscale, texture=&LowResBlurredNoise}
  n14 = LinearInterpolate(A=ScalarParameter() {defaultValue=0.3, parameterName=Min}, Alpha=Multiply(A=@n7, B=@n13.G)) {constA=0.4}
  n15 = ComponentMask(None=TextureCoordinate()) {g=true}
  n16 = ScalarParameter() {defaultValue=15, parameterName=ColorLerpPower}
  n17 = LinearInterpolate(Alpha=LinearInterpolate(A=@n7, Alpha=@n13.G)) {constA=2, constB=0.9}
  n18 = Multiply(A=@n14, B=Power(Base=@n15, Exp=Multiply(A=@n16, B=@n17)) {constExponent=4})
  n19 = Multiply(
   A=LinearInterpolate(
    A=LinearInterpolate(A=@n1.RGB, B=@n2.RGB, Alpha=@n7)
    B=VectorParameter() {parameterName=BottomColor, defaultValue=<rgba 0.3 1.8406 2 0>}.RGB
    Alpha=Multiply(A=@n9, B=@n18)
   )
   B=ScalarParameter() {defaultValue=1.5, parameterName=Intensity}
  )
  n20 = ScalarParameter() {defaultValue=5, parameterName=HDR Intensity}
  n21 = VectorParameter() {parameterName=NoiseTexScale2, defaultValue=<rgba 3 4 0 0>}
  n22 = VertexColor()
  n23 = Multiply(A=TextureCoordinate(), B=AppendVector(A=LinearInterpolate(A=@n21.R, B=@n21.R, Alpha=@n22.B), B=Constant()))
  n24 = Multiply(A=Time(), B=ScalarParameter() {defaultValue=0.25, parameterName=WorldAlignedPanSpeed})
  n25 = ComponentMask(None=TextureSample(UVs=Panner(Coordinate=@n23, Time=@n24) {speedY=1}) {texture=&T_Smooth_tile}.RGB) {g=true}
  MaterialFunctionCall(Emissive=MaterialFunctionCall(LDR Input=Clamp(None=@n19), HDR Tint=@n19, HDR Intensity Multiplier=LinearInterpolate(B=@n20, Alpha=@n25) {constA=1}) {materialFunction=&DeriveHDRfromLDR}.RemappedRange) {materialFunction=&MF_FX_EmissiveAdjustments}
  n26 = ScalarParameter() {defaultValue=1.25, parameterName=BottomHeightVariance}
  n27 = SphereMask(
   A=ComponentMask(None=TextureCoordinate() {vTiling=0.5}) {g=true}
   B=Constant()
   Radius=Multiply(B=LinearInterpolate(A=@n26, Alpha=@n13.G) {constA=0.95}) {constA=0.5}
   Hardness=ScalarParameter() {defaultValue=0.91, parameterName=MaskHardness}
  )
  n28 = Power(Base=Power(Base=Multiply(A=@n15, B=@n27)), Exp=@n17)
  n29 = Sine(None=ComponentMask(None=TextureCoordinate()) {r=true}) {period=2}
  n30 = Multiply(A=@n29, B=@n29)
  Multiply(A=ScalarParameter() {defaultValue=0.5, parameterName=Opacity}, B=Multiply(A=@n14, B=Multiply(A=@n28, B=@n30)))
  n31 = Constant3Vector() {constant=<rgba -1 0 0 0>}
  n32 = Constant3Vector() {constant=<rgba 1 0 0 0>}
  n33 = TextureCoordinate()
  n34 = VectorParameter() {parameterName=NoiseTexScale4, defaultValue=<rgba 2 2 0 0>}
  n35 = AppendVector(A=LinearInterpolate(A=@n34.R, B=@n34.R, Alpha=VertexColor().B), B=Constant())
  n36 = Time()
  n37 = ScalarParameter() {defaultValue=0.25, parameterName=2WorldAlignedPanSpeed}
  n38 = TextureSample(UVs=Panner(Coordinate=Multiply(A=@n33, B=@n35), Time=Multiply(A=@n36, B=@n37)) {speedY=1}) {texture=&T_Smooth_tile}
  Transform(None=Multiply(A=LinearInterpolate(A=@n31, B=@n32, Alpha=@n38.R), B=ScalarParameter() {defaultValue=1, parameterName=WPOScale})) {transformSourceType=TRANSFORMSOURCE_Local}
 ]
}
